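# Material function that converts an RGB input to a four-channel CMYK output, with a threshold parameter grouped under Comic Dots for a post-process effect.
# Black is one minus the largest saturated RGB channel, while CMY is simply the inverse of RGB, without normalization by black.
# A branchless Sign and Saturate mask zeroes CMY where black reaches the CMYK Black Threshold, then black is appended as the fourth channel.
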MaterialFunction /Game/Creative/PostProcess/MaterialFunctions/MF_RGBConversionToCMYK
graph {nodes [
  n1 = Saturate(None=FunctionInput() {inputName=RGB})
  n2 = OneMinus(None=@n1)
  n3 = ScalarParameter() {defaultValue=0.9, group=2. Comic Dots, parameterName=CMYK Black Threshold}
  n4 = ComponentMask(None=@n1) {g=true}
  n5 = ComponentMask(None=@n1) {b=true}
  n6 = OneMinus(None=Max(A=ComponentMask(None=@n1) {r=true}, B=Max(A=@n4, B=@n5)))
  n7 = Saturate(None=Sign(None=Subtract(A=@n3, B=@n6)))
  FunctionOutput(None=AppendVector(A=Multiply(A=@n2, B=@n7), B=@n6)) {outputName=CMYK}
 ]}
