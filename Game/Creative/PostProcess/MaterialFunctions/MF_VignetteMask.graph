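# MF_VignetteMask is a material function that outputs a radial vignette mask for post-process materials.
# It measures the distance from the texture coordinates to the screen center at (0.5, 0.5).
# An inverse lerp remaps that distance between the Vignette Inner Radius parameter and a fixed outer edge of 0.5, then saturates it, so the mask ramps from zero inside to one toward the edges.
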MaterialFunction /Game/Creative/PostProcess/MaterialFunctions/MF_VignetteMask
graph {nodes [
  n1 = ScalarParameter() {defaultValue=0.333, group=Vignette, parameterName=Vignette Inner Radius, sliderMax=0.49}
  n2 = Constant() {r=0.5}
  n3 = Distance(A=TextureCoordinate(), B=Constant2Vector() {g=0.5, r=0.5})
  FunctionOutput(None=Saturate(None=InverseLinearInterpolate(A=@n1, B=@n2, Value=@n3))) {outputName=VignetteMask}
 ]}
